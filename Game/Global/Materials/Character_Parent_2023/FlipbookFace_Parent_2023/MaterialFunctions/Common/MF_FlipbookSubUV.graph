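# Material function that samples one cell of a flipbook texture grid by Frame and SubImages, in the flipbook face parent material set.
# It floors Frame, derives column and row, then offsets and scales UVs.
# RecordTextureStreamingInfo receives UVs divided by two to the floored Streaming MIP Bias, steering mip streaming.
# Despite its description, it blends no frames, outputting RGB, Alpha and the unmodified input UVs.
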MaterialFunction /Game/Global/Materials/Character_Parent_2023/FlipbookFace_Parent_2023/MaterialFunctions/Common/MF_FlipbookSubUV
graph {
 properties {
  bExposeToLibrary=true
  description=Allows SubUV blending of multiple texture frames in a single texture
  libraryCategoriesText [Misc,Texturing]
 }
 nodes [
  n1 = FunctionInput() {inputName=Frame, inputType=FunctionInput_Scalar, sortPriority=-10}
  n2 = Subtract(A=@n1, B=Frac(None=@n1) {desc=Frame Lerp}) {desc=IndexA}
  n3 = FunctionInput() {inputName=SubImages, inputType=FunctionInput_Vector2, sortPriority=-20, previewValue=<2 2 0 0>}
  n4 = ComponentMask(None=@n3) {desc="## Horz Images", r=true}
  n5 = Divide(A=Constant2Vector() {g=1, r=1}, B=@n3) {desc=Scale}
  n6 = Multiply(A=@n2, B=ComponentMask(None=@n5) {desc="1 / ## Horz Images", r=true})
  n7 = FunctionInput(Preview=TextureCoordinate()) {bUsePreviewValueAsDefault=true, inputName=UVs, inputType=FunctionInput_Vector2, sortPriority=-30, previewValue=<2 2 0 0>}
  n8 = Add(A=AppendVector(A=Fmod(A=@n2, B=@n4) {desc=IndexA_Horz}, B=Floor(None=@n6) {desc=IndexA_Vert}) {desc=IndexA_Combined}, B=@n7)
  n9 = TextureObject() {texture=/Engine/EngineMaterials/DefaultDiffuse}
  n10 = NamedRerouteUsage()
  n11 = Power(Base=Constant() {r=2}, Exp=Floor(None=ScalarParameter() {group=Flipbook Streaming MIP Bias, parameterName=Streaming MIP Bias}))
  n12 = TextureSample(
   UVs=NamedRerouteDeclaration(None=Multiply(A=@n8, B=@n5) {desc=UV_A}) {name=UVs, nodeColor=<rgba 0.4297 1 0 1>}
   Tex=RecordTextureStreamingInfo!(
    Tex=FunctionInput(Preview=@n9) {bUsePreviewValueAsDefault=true, inputName=Texture, inputType=FunctionInput_Texture2D, sortPriority=-40}
    UVs=Divide(A=@n10, B=@n11)
   ).Tex
  ) {texture=&DefaultDiffuse}
  FunctionOutput(None=@n12.RGB) {outputName=RGB}
  FunctionOutput(None=@n12.A) {outputName=Alpha}
  FunctionOutput(None=@n7) {outputName=UVs, sortPriority=3}
 ]
}
