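# Material function for flipbook character faces that turns a distance-field input into an antialiased line mask.
# Line thickness blends between near and far parameters, apparently via an FOV-compensated distance fade shared from the Anime parent functions.
# The inverted thickness is the threshold, and its difference from the field is divided by screen-space derivative length plus a softness-controlled bias, keeping edges smooth at varying screen sizes.
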MaterialFunction /Game/Global/Materials/Character_Parent_2023/FlipbookFace_Parent_2023/MaterialFunctions/Common/MF_Flipbook_DFLine
graph {nodes [
  n1 = ScalarParameter() {defaultValue=0.1, desc="LineThickness (Near)", group=Flipbook 01 Global, parameterName=FBDF_LineThickness, sliderMax=1, sortPriority=500}
  n2 = ScalarParameter() {defaultValue=0.7, group=Flipbook 01 Global, parameterName=FBDF_LineThicknessFar, sliderMax=2, sortPriority=501}
  n3 = NamedRerouteUsage()
  n4 = Subtract(A=OneMinus(None=LinearInterpolate(A=@n1, B=@n2, Alpha=@n3)), B=FunctionInput() {inputName=DF, inputType=FunctionInput_Scalar})
  n5 = AppendVector(A=DDX(Value=@n4), B=DDY(Value=@n4))
  n6 = ScalarParameter() {defaultValue=0.9, group=Flipbook 01 Global, parameterName=FBDF_LineSoftness, sortPriority=510}
  n7 = Divide(A=@n4, B=Add(A=Length(None=@n5) {is=class MaterialExpressionLength}, B=LinearInterpolate(Alpha=@n6) {constA=0.5, constB=0}) {constB=0})
  FunctionOutput(None=Saturate(None=Subtract(B=@n7) {constA=0.5}))
  n8 = ScalarParameter() {defaultValue=1000, group=Flipbook 01 Global, parameterName=FBDF_FadeDistance, sliderMax=1500, sliderMin=500, sortPriority=503}
  NamedRerouteDeclaration(None=MaterialFunctionCall!(
    Fade Offset=ScalarParameter() {defaultValue=50, group=Flipbook 01 Global, parameterName=FBDF_FadeOffset, sliderMax=100, sortPriority=502}
    Fade Distance=@n8
   ) {materialFunction=/Game/Global/Materials/Character_Parent_2023/Anime_Parent_2023/MaterialFunctions/MFC_DistanceFade_FOV_compensation}.Result) {name=DistanceFading, nodeColor=<rgba 1 0.3047 0 1>}
 ]}
